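# Physics asset for the Nautical Pajamas Night backpack cosmetic, defining simple box bodies and joints for its straw and note bones.
# Two chains hang off the small pack body: a four-segment straw and a two-segment note.
# The straw has a looser base joint and tight limits between later segments.
# The note joints favor swing on one axis, with twist locked where it meets the pack.
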PhysicsAsset /BRCosmetics/Accessories/FORT_Backpacks/Backpack_M_MED_Nautical_Pajamas_Night/Meshes/M_MED_Nautical_Pajamas_Night_Pack_Physics
bodies {
 note1 [{shapeType=Box, extentX=4, extentY=6, extentZ=1, center=<2 -1.0279 0.3531>}]
 note2 [{shapeType=Box, extentX=3, extentY=6.4849, extentZ=0.412, center=<1.5 0.1654 -0.0715>}]
 pack [{shapeType=Box, extentX=0.5, extentY=0.5, extentZ=0.5, center=<-1.4794 -5.3921 0.2231>}]
 straw1 [{shapeType=Box, extentX=6.7201, extentY=1.0577, extentZ=1.0855, center=<1.3082 0.0059 0.0019>}]
 straw2 [{shapeType=Box, extentX=0.5, extentY=1, extentZ=1, center=<0.25 0 0>}]
 straw3 [{shapeType=Box, extentX=0.5, extentY=1, extentZ=1, center=<0.25 0 0>}]
 straw4 [{shapeType=Box, extentX=4.5, extentY=1.5, extentZ=1, center=<2.25 -0.1757 -0.0663>}]
}
constraints [
 {bone1Name=straw1, bone2Name=pack, swing1LimitDegrees=10, swing2LimitDegrees=10, twistLimitDegrees=20}
 {bone1Name=straw2, bone2Name=straw1, swing1LimitDegrees=5, swing2LimitDegrees=5, twistLimitDegrees=2}
 {bone1Name=straw3, bone2Name=straw2, swing1LimitDegrees=5, swing2LimitDegrees=5, twistLimitDegrees=2}
 {bone1Name=straw4, bone2Name=straw3, swing1LimitDegrees=5, swing2LimitDegrees=5, twistLimitDegrees=2}
 {bone1Name=note1, bone2Name=pack, swing1LimitDegrees=2, swing2LimitDegrees=25, twistLimitDegrees=15, twistMotion=Locked}
 {bone1Name=note2, bone2Name=note1, swing1LimitDegrees=5, swing2LimitDegrees=15, twistLimitDegrees=5}
]
