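# Physics asset for the Nautical Pajamas Night backpack cosmetic, covering its pack, cup and straw bones.
# A small pack box anchors a larger cup box, with seven capsules chained end to end from the cup along the straw.
# Each straw joint has very tight swing and twist limits, while the cup-to-pack joint allows slightly more, mostly in twist, suggesting a subtle, nearly rigid wobble.
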PhysicsAsset /BRCosmetics/Accessories/FORT_Backpacks/Backpack_M_MED_Nautical_Pajamas_Night/Meshes/M_MED_Nautical_Pajamas_Night_Physics
bodies {
 cup [{shapeType=Box, extentX=18.5783, extentY=12.7118, extentZ=12.725, center=<0.046 -5.558 0.6082>}]
 pack [{shapeType=Box, extentX=1.01, extentY=1.01, extentZ=1.01}]
 straw1 [{shapeType=Capsule, length=11.4017, radius=0.6017, center=<1.3214 -0.0562 -0.0896>, rotation=<rot 90 0 0>}]
 straw2 [{shapeType=Capsule, length=1.0889, radius=0.6066, center=<1.1132 -0.0556 -0.2624>, rotation=<rot 90 0 0>}]
 straw3 [{shapeType=Capsule, length=1.541, radius=0.8017, center=<1.211 -0.0273 -0.1905>, rotation=<rot 90 0 0>}]
 straw4 [{shapeType=Capsule, length=1.3352, radius=0.6775, center=<0.8276 -0.0001 -0.081>, rotation=<rot 90 0 0>}]
 straw5 [{shapeType=Capsule, length=0.9627, radius=0.6224, center=<0.8165 -0.0094 -0.1613>, rotation=<rot 90 0 0>}]
 straw6 [{shapeType=Capsule, length=2.3379, radius=0.7626, center=<2.109 -0.0127 -0.3278>, rotation=<rot 90 0 0>}]
 straw7 [{shapeType=Capsule, length=5.5531, radius=0.6281, center=<3.407 0.0821 -0.114>, rotation=<rot 90 0 0>}]
}
constraints [
 {bone1Name=cup, bone2Name=pack, swing1LimitDegrees=1, swing2LimitDegrees=2, twistLimitDegrees=7}
 {bone1Name=straw1, bone2Name=cup, swing1LimitDegrees=1, swing2LimitDegrees=1, twistLimitDegrees=1}
 {bone1Name=straw2, bone2Name=straw1, swing1LimitDegrees=1, swing2LimitDegrees=1, twistLimitDegrees=1}
 {bone1Name=straw3, bone2Name=straw2, swing1LimitDegrees=1, swing2LimitDegrees=1, twistLimitDegrees=1}
 {bone1Name=straw4, bone2Name=straw3, swing1LimitDegrees=1, swing2LimitDegrees=1, twistLimitDegrees=1}
 {bone1Name=straw5, bone2Name=straw4, swing1LimitDegrees=1, swing2LimitDegrees=1, twistLimitDegrees=1}
 {bone1Name=straw6, bone2Name=straw5, swing1LimitDegrees=1, swing2LimitDegrees=1, twistLimitDegrees=1}
 {bone1Name=straw7, bone2Name=straw6, swing1LimitDegrees=1, swing2LimitDegrees=1, twistLimitDegrees=1}
]
